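# Sound cue for the Demon Hunter crossbow impact on a player in Athena.
# A Random node picks one of four impact wave variants with equal weight and feeds a Modulator that varies playback per trigger.
# It uses the bullet impact attenuation and the impacts sound class, with a concurrency override that caps simultaneous instances and stops the oldest.
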SoundCue /Game/Athena/Sounds/Impacts/Crossbow_DemonHunter_Player_Impact_Athena_Cue
root=Modulator_0
nodes [
 {id=Modulator_0, children [Random_1], properties {volumeMax=1.5}}
 {id=Random_1, children [WavePlayer_0,WavePlayer_1,WavePlayer_2,WavePlayer_3], properties {weights [1,1,1,1]}}
 {id=WavePlayer_0, properties {soundWaveAssetPtr=&/Crossbow_DemonHunter_Player_Impact_04}}
 {id=WavePlayer_1, properties {soundWaveAssetPtr=&/Crossbow_DemonHunter_Player_Impact_01}}
 {id=WavePlayer_2, properties {soundWaveAssetPtr=&/Crossbow_DemonHunter_Player_Impact_02}}
 {id=WavePlayer_3, properties {soundWaveAssetPtr=&/Crossbow_DemonHunter_Player_Impact_03}}
]
properties {
 attenuationSettings=/Game/Sounds/Attenuation/Fort_Bullet_Impact_Att
 bOverrideConcurrency=true
 maxDistance=5300
 soundClassObject=&Fort_Impacts
 volumeMultiplier=2.2
 concurrencyOverrides {maxCount=3, resolutionRule=StopOldest}
}
