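# Sound cue for crossbow impacts on players in Athena, using arrow-into-flesh sounds.
# A Random node picks among six equally weighted flesh impact waves and feeds a Modulator that varies playback, with a raised volume ceiling.
# It uses the shared bullet impact attenuation and impacts sound class, with a concurrency override that stops the oldest instance once a few are playing.
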SoundCue /Game/Athena/Sounds/Impacts/Crossbow_Player_Impact_Athena_Cue
root=Modulator_0
nodes [
 {id=Modulator_0, children [Random_0], properties {volumeMax=1.5}}
 {id=Random_0, children [WavePlayer_0,WavePlayer_1,WavePlayer_2,WavePlayer_3,WavePlayer_8,WavePlayer_9], properties {weights [1,1,1,1,1,1]}}
 {id=WavePlayer_0, properties {soundWaveAssetPtr=&/Arrow_Default_Impact_Flesh_01}}
 {id=WavePlayer_1, properties {soundWaveAssetPtr=&/Arrow_Default_Impact_Flesh_02}}
 {id=WavePlayer_2, properties {soundWaveAssetPtr=&/Arrow_Default_Impact_Flesh_03}}
 {id=WavePlayer_3, properties {soundWaveAssetPtr=&/Arrow_Default_Impact_Flesh_04}}
 {id=WavePlayer_8, properties {soundWaveAssetPtr=&/Arrow_Default_Impact_Flesh_05}}
 {id=WavePlayer_9, properties {soundWaveAssetPtr=&/Arrow_Default_Impact_Flesh_06}}
]
properties {
 attenuationSettings=/Game/Sounds/Attenuation/Fort_Bullet_Impact_Att
 bOverrideConcurrency=true
 maxDistance=5300
 soundClassObject=&Fort_Impacts
 volumeMultiplier=1.3
 concurrencyOverrides {maxCount=3, resolutionRule=StopOldest}
}
