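# Sound cue for crossbow impacts on surfaces in Athena.
# A random node picks among four equally weighted impact waves, and its output passes through a modulator node for variation.
# The cue uses the shared bullet impact attenuation and the impacts sound class, with a boosted volume multiplier.
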SoundCue /Game/Athena/Sounds/Impacts/Crossbow_Surface_Impact_Athena_Cue
root=Modulator_0
nodes [
 {id=Modulator_0, children [Random_0]}
 {id=Random_0, children [WavePlayer_1,WavePlayer_2,WavePlayer_3,WavePlayer_4], properties {weights [1,1,1,1]}}
 {id=WavePlayer_1, properties {soundWaveAssetPtr=&/sfx_crossbow_impact_01}}
 {id=WavePlayer_2, properties {soundWaveAssetPtr=&/sfx_crossbow_impact_02}}
 {id=WavePlayer_3, properties {soundWaveAssetPtr=&/sfx_crossbow_impact_03}}
 {id=WavePlayer_4, properties {soundWaveAssetPtr=&/sfx_crossbow_impact_04}}
]
properties {
 attenuationSettings=/Game/Sounds/Attenuation/Fort_Bullet_Impact_Att
 maxDistance=5300
 soundClassObject=&Fort_Impacts
 volumeMultiplier=5.2416
}
