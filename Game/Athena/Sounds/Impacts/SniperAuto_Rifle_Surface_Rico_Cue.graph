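# SoundCue for automatic sniper rifle surface ricochet impacts in Athena.
# A Random node picks one of four ricochet waves with equal weight and passes it through a Modulator.
# The cue plays in the impacts sound class with a boosted volume, using the shared bullet impact attenuation.
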SoundCue /Game/Athena/Sounds/Impacts/SniperAuto_Rifle_Surface_Rico_Cue
root=Modulator_0
nodes [
 {id=Modulator_0, children [Random_1]}
 {id=Random_1, children [WavePlayer_0,WavePlayer_5,WavePlayer_6,WavePlayer_7], properties {weights [1,1,1,1]}}
 {id=WavePlayer_0, properties {soundWaveAssetPtr=&/SniperAuto_Rico_01}}
 {id=WavePlayer_5, properties {soundWaveAssetPtr=&/SniperAuto_Rico_02}}
 {id=WavePlayer_6, properties {soundWaveAssetPtr=&/SniperAuto_Rico_03}}
 {id=WavePlayer_7, properties {soundWaveAssetPtr=&/SniperAuto_Rico_04}}
]
properties {
 attenuationSettings=/Game/Sounds/Attenuation/Fort_Bullet_Impact_Att
 maxDistance=5300
 soundClassObject=&Fort_Impacts
 volumeMultiplier=2.016
}
