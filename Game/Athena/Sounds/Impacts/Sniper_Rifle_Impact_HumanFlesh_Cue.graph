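# Sound cue for sniper rifle bullet impacts on human flesh.
# A random node picks one of four equally weighted body impact waves, and a modulator adds slight upward pitch variation.
# It plays in the impacts sound class with bullet impact attenuation and a boosted volume.
# A concurrency override limits simultaneous instances and stops the oldest when exceeded.
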SoundCue /Game/Athena/Sounds/Impacts/Sniper_Rifle_Impact_HumanFlesh_Cue
root=Modulator_0
nodes [
 {id=Modulator_0, children [Random_0], properties {pitchMax=1.1, pitchMin=1, volumeMax=1, volumeMin=1}}
 {
  id=Random_0
  children [WavePlayer_4,WavePlayer_5,WavePlayer_6,WavePlayer_7]
  properties {bShouldExcludeFromBranchCulling=true, weights [1,1,1,1]}
 }
 {id=WavePlayer_4, properties {soundWaveAssetPtr=&/Sniper_Impact_Body_02}}
 {id=WavePlayer_5, properties {soundWaveAssetPtr=&/Sniper_Impact_Body_03}}
 {id=WavePlayer_6, properties {soundWaveAssetPtr=&/Sniper_Impact_Body_04}}
 {id=WavePlayer_7, properties {soundWaveAssetPtr=&/Sniper_Impact_Body_01}}
]
properties {
 attenuationSettings=/Game/Sounds/Attenuation/Fort_Bullet_Impact_Att
 bOverrideConcurrency=true
 maxDistance=5300
 soundClassObject=&Fort_Impacts
 volumeMultiplier=4.248
 concurrencyOverrides {maxCount=3, resolutionRule=StopOldest}
}
