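# Sound cue for sniper rifle bullet impacts on metal surfaces.
# A Random node picks one of five equally weighted metal impact waves and is excluded from branch culling, then feeds a Modulator at the root for playback variation.
# The cue boosts volume, plays in the impacts sound class and uses the shared bullet impact attenuation.
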SoundCue /Game/Athena/Sounds/Impacts/Sniper_Rifle_Impact_Metal_Cue
root=Modulator_0
nodes [
 {id=Modulator_0, children [Random_0]}
 {
  id=Random_0
  children [WavePlayer_1,WavePlayer_2,WavePlayer_3,WavePlayer_4,WavePlayer_9]
  properties {bShouldExcludeFromBranchCulling=true, weights [1,1,1,1,1]}
 }
 {id=WavePlayer_1, properties {soundWaveAssetPtr=&/Sniper_Impact_Metal_01}}
 {id=WavePlayer_2, properties {soundWaveAssetPtr=&/Sniper_Impact_Metal_02}}
 {id=WavePlayer_3, properties {soundWaveAssetPtr=&/Sniper_Impact_Metal_03}}
 {id=WavePlayer_4, properties {soundWaveAssetPtr=&/Sniper_Impact_Metal_04}}
 {id=WavePlayer_9, properties {soundWaveAssetPtr=&/Sniper_Impact_Metal_05}}
]
properties {
 attenuationSettings=/Game/Sounds/Attenuation/Fort_Bullet_Impact_Att
 maxDistance=5300
 soundClassObject=&Fort_Impacts
 volumeMultiplier=3.5683
}
